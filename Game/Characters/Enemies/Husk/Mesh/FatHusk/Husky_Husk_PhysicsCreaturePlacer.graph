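# Physics asset for the fat Husk enemy skeleton, named for use as a creature placer.
# It holds a single capsule body on the pelvis, rotated and slightly offset along the bone's X axis, as a coarse volume for the creature.
# It has no constraints, so it defines no articulated ragdoll.
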PhysicsAsset /Game/Characters/Enemies/Husk/Mesh/FatHusk/Husky_Husk_PhysicsCreaturePlacer
bodies {pelvis [{shapeType=Capsule, length=69.75, radius=50.25, center=<-5.7429 0 0>, rotation=<rot -90 -0.0198 0.0198>}]}
constraints []
